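# Animation Blueprint for the RageDebris pickaxe cosmetic's axe skeleton.
# The anim graph passes a sub-input pose straight to the output pose with no added nodes.
# The event graph has an unconnected update event.
# Gravity override variables are declared but not used in either graph.
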AnimBlueprint /BRCosmetics/Weapons/FORT_Melee/Pickaxe_RageDebris/Meshes/RageDebris_Axe_AnimBP
target_skeleton=&/RageDebris_Axe_Skeleton
graphs [
 AnimGraph {nodes [
   AnimGraphNode_SubInput_0 = Input Pose()
   Output Pose(Result=@AnimGraphNode_SubInput_0) {layerGroup=None}
  ]}
 EventGraph {nodes [Event BlueprintUpdateAnimation()]}
]
variables [
 Gravity Override {type=Vector}
 Gravity Override Params {type=GravityOverrideParamsStruct}
]
